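# Material for the Capture T2 vehicle cosmetic wheel, adding an animated emissive glow on skeletal meshes.
# MF_WheelParent_Simple supplies the base attributes and the rim and tire masks.
# A gradient texture panning along U multiplies each mask's red channel, vertex color red blends rim to tire, and the result is added to emissive.
# The attributes then pass through the shared vehicle damage and de-rez transition functions.
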Material /VehicleCosmetics/Wheels/Capture_T2/Materials/M_Wheel_Capture_T2
graph {
 properties {
  bAutomaticallySetUsageInEditor=false
  bUseMaterialAttributes=true
  bUsedWithMeshDeformer=true
  bUsedWithSkeletalMesh=true
  pixelDepthOffsetMode=PDOM_Legacy
  usedShadingModels=Default Lit
 }
 nodes [
  n1 = MaterialFunctionCall() {materialFunction=&MF_WheelParent_Simple}
  n2 = GetMaterialAttributes(None=@n1.MaterialAttributes) {attributeGetTypes [#1]}
  n3 = ComponentMask(None=@n1.RimRGBAMask) {r=true}
  n4 = TextureSample(UVs=Panner() {speedX=-1}) {texture=/VehicleCosmetics/Wheels/_SharedTextures/FX/GradMask_T}
  n5 = ComponentMask(None=@n1.TireRGBAMask) {r=true}
  n6 = Add(A=@n2.Emissive Color, B=LinearInterpolate(A=Multiply(A=@n3, B=@n4.G), B=Multiply(A=@n5, B=@n4.G), Alpha=@n1.VertexColorR))
  MaterialFunctionCall(In=MaterialFunctionCall(Material=SetMaterialAttributes(MaterialAttributes=@n2.MaterialAttributes, Emissive Color=@n6) {attributeSetTypes [#1]}) {materialFunction=&MF_BaseVehicle_Damage}.Result) {materialFunction=&MF_VehicleDeRez_Transition}
 ]
}
