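# Translucent unlit VFX material for the Capture T2 vehicle cosmetic wheel, drawing an animated glowing flare on skeletal meshes.
# Emissive is a brightened CustomColor parameter, gold by default, while textures shape only the opacity.
# A steadily spinning flare with adjustable exponent is added to a second flare plus faint starburst detail, whose rotation swings back and forth on a sine of time.
# A static soft circle texture masks the sum.
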Material /VehicleCosmetics/Wheels/Capture_T2/Materials/M_Wheel_Capture_T2_VFX
graph {
 properties {
  bAutomaticallySetUsageInEditor=false
  bUseMaterialAttributes=true
  bUsedWithMeshDeformer=true
  bUsedWithSkeletalMesh=true
  blendMode=BLEND_Translucent
  previewMesh=/VehicleCosmetics/Wheels/Capture_T2/Meshes/SK_Capture_Tier2
  shadingModel=MSM_Unlit
  usedShadingModels=Unlit
 }
 nodes [
  n1 = Power(
   Base=TextureSample(UVs=Rotator() {speed=-6}) {texture=/VehicleCosmetics/Wheels/Capture_T2/Textures/Flare_Headlight01_D}.R
   Exp=ScalarParameter() {defaultValue=0.7, parameterName=PowerExponent}
  )
  n2 = Rotator(Time=Sine(None=Time()) {period=4}) {speed=-2}
  n3 = Add(
   A=Multiply(
    A=TextureSample(UVs=@n2) {texture=/VehicleCosmetics/Wheels/Capture_T2/Textures/StarBurst_01_T}.G
    B=ScalarParameter() {defaultValue=0.02, parameterName=DetailMultiply}
   )
   B=TextureSample(UVs=@n2) {texture=&Flare_Headlight01_D}.R
  )
  SetMaterialAttributes!(
   MaterialAttributes
   Emissive Color=Multiply(B=VectorParameter() {parameterName=CustomColor, defaultValue=<rgba 0.8469 0.6038 0.1384 1>}.RGB) {constA=6}
   Opacity=Multiply(A=Add(A=@n1, B=@n3), B=TextureSample() {texture=/VehicleCosmetics/Wheels/_SharedTextures/FX/SoftCircle01_D}.R)
  ) {attributeSetTypes [#1, #2]}
 ]
}
